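# Translucent unlit Niagara sprite material for sand smoke on the Maroon Male glider effects.
# Two noise layers on camera-aligned virtual-plane UVs crossfade by a sine of time, modulating a center color.
# Dynamic parameters drive hue shift, desaturation, scale and dissolve, with a time-of-day curve tint and separate tiling-dust highlight.
# Opacity math combines a hard-edged dissolve threshold, puffy smoke texture, particle alpha and depth fade.
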Material /Game/Gadgets/Assets/VinderTech_GliderChute/Glider_Ensemble_Maroon_Male/FX/Materials/M_SandSmoke
graph {
 properties {bUsedWithNiagaraSprites=true, blendMode=BLEND_Translucent, pixelDepthOffsetMode=PDOM_Legacy, shadingModel=MSM_Unlit, usedShadingModels=Unlit}
 nodes [
  n1 = Time() {bOverride_Period=true, period=500}
  n2 = Multiply(A=@n1, B=ScalarParameter() {defaultValue=0.5, group=35.TextureBlend, parameterName=Trd_ExtraLayerBlendSpeed})
  n3 = LinearInterpolate(A=NamedRerouteUsage(), B=NamedRerouteUsage(), Alpha=ConstantBiasScale(None=Sine(None=@n2)))
  n4 = LinearInterpolate(A=ScalarParameter() {defaultValue=0.65, group=35.TextureBlend, parameterName=Trd_RemapLow}, Alpha=@n3)
  Add(A=@n4, B=@n3)
  n5 = TextureObjectParameter() {parameterName=DissolveTexture, texture=/Game/CharClasses/Outlander/Gadgets/AntiMaterialCharge/Materials/T_AntiMaterialCharge_Dissolve_01}
  n6 = VectorParameter() {parameterName=UVControl, defaultValue=<rgba 1 0.5 0.2 0>}
  n7 = MaterialFunctionCall(Texture=@n5, UVControl_RGB=@n6.RGB, UVControl_A=@n6.RGB) {materialFunction=&FXCommon_TextureSample}
  n8 = DynamicParameter() {defaultValue=<rgba 0 1 1 1>, paramNames [Dissolve,Deast,Scale,Hue]}
  n9 = ScalarParameter() {parameterName=DissolveTest}
  n10 = Constant() {r=1}
  If(A=Subtract(A=@n7.R, B=Add(A=@n8.Dissolve, B=@n9)), B=Constant(), "A > B"=@n10, "A == B"=@n10, "A < B"=Constant())
  n11 = CurveAtlasRowParameter(CurveTime=CollectionParameter() {collection=&FortniteMaterialParameters, parameterName=TimeOfDay_0-1}) {
   atlas=&AthenaCurveAtlas01
   curve=/Game/Athena/LinearColorCurves/Curve_TimeOfDay_S23
   defaultValue=321
   group=TimeOfDay
   parameterName=TOD_Curve
  }
  n12 = DynamicParameter() {defaultValue=<rgba 0 1 1 1>, paramNames [Dissolve,Deast,Scale,Hue]}
  n13 = VectorParameter() {parameterName=Color_CenterColor, defaultValue=<rgba 0.25 0.1485 0.0898 1>}
  n14 = Multiply(
   A=@n11
   B=MaterialFunctionCall(Hue Shift Percentage=@n12.Hue, Texture=Multiply(A=@n13.RGB, B=@n4)) {materialFunction=&HueShift}.Result
  )
  n15 = Multiply(A=Desaturation(None=@n14, Fraction=@n12.Deast), B=@n12.Scale)
  MaterialFunctionCall(Emissive=@n15) {materialFunction=&MF_FX_EmissiveAdjustments}
  n16 = VectorParameter() {parameterName=LightSand_UV}
  n17 = MaterialFunctionCall(
   Texture=TextureObject() {texture=/Game/Global/Textures/Default/PointSplatter/T_Tiling_Dust}
   UVControl_RGB=@n16.RGB
   UVControl_A=@n16.RGB
   "[Override]TimeInput"=@n1
   "[Override]UVInput"=NamedRerouteUsage()
  ) {materialFunction=&FXCommon_TextureSample}
  n18 = ScalarParameter() {defaultValue=1, parameterName=LightSand_Power}
  Multiply(A=@n3, B=Multiply(A=Power(Base=@n17.R, Exp=@n18), B=VectorParameter() {parameterName=L_Color, defaultValue=<rgba 0 0 0 1>}.RGB))
  n19 = Subtract(A=@n7.R, B=Add(A=Multiply(A=@n8.Dissolve) {constB=2}, B=@n9))
  n20 = Saturate(None=Multiply(A=Saturate(None=@n19), B=ScalarParameter() {defaultValue=100, parameterName=DissolveHardness}))
  n21 = TextureSampleParameter2D() {parameterName=SmokeTexture, texture=/Game/Effects/Fort_Effects/Textures/Smoke/T_SmokePuffy_SubUV_2X2}
  Multiply(
   A=Multiply(A=ParticleColor().A, B=Multiply(A=@n20, B=@n21.R))
   B=DepthFade(FadeDistance=ScalarParameter() {group=35.TextureBlend, parameterName=DepthFadeDistance})
  )
  n22 = NamedRerouteUsage()
  n23 = VectorParameter() {group=32.TextureTex_Trd, parameterName=TextureTex_Trd_UVControl_ExtraLayer, defaultValue=<rgba 0.015 0.015 0 0>}
  n24 = Transform(None=Constant3Vector() {constant=<rgba 0 0 -1 0>}) {transformSourceType=TRANSFORMSOURCE_Camera}
  n25 = Transform(None=Constant3Vector() {constant=<rgba 1 0 0 0>}) {transformSourceType=TRANSFORMSOURCE_Camera}
  n26 = ObjectPositionWS()
  n27 = NamedRerouteDeclaration(None=MaterialFunctionCall(
    UVs=MaterialFunctionCall(Plane Normal Axis=@n24, Plane X Axis=@n25, Plane Center=@n26) {materialFunction=&VirtualPlaneCoordinates}.UVs
    Texture Scale=ScalarParameter() {defaultValue=1, group=00.Base, parameterName=Base_TextureTexUV}
   ) {materialFunction=&ScaleUVsByCenter}.UVs) {name=ScreenUV, nodeColor=<rgba 0 0.3203 1 1>}
  n28 = TextureObjectParameter() {
   group=32.TextureTex_Trd
   parameterName=TextureTex_Trd_Tex
   samplerType=SAMPLERTYPE_LinearGrayscale
   texture=/Game/Global/Textures/Default/Noise/T_WhiteNoise
  }
  NamedRerouteDeclaration(None=TextureSample!(
    UVs=MaterialFunctionCall!(UVInput=@n27, TimeInput=@n22, Control_RGB=@n23.RGB, Control_A=@n23.RGB) {materialFunction=/Game/Effects/Niagara/MasterMaterials/Functions/Common/FXCommon_QuickUVControl}.UV
    Tex=@n28
    Level
    Apply View MipBias
   ) {constMipValue=0, mipValueMode=TMVM_MipLevel, samplerType=SAMPLERTYPE_LinearColor}.R) {name=TextureTex_Trd_ExtraLayer, nodeColor=<rgba 0.3125 1 0 1>}
  n29 = VectorParameter() {group=32.TextureTex_Trd, parameterName=TextureTex_Trd_UVControl, defaultValue=<rgba 0.013 0.013 0 0>}
  NamedRerouteDeclaration(None=TextureSample(
    UVs=MaterialFunctionCall(UVInput=@n27, TimeInput=@n22, Control_RGB=@n29.RGB, Control_A=@n29.RGB) {materialFunction=&FXCommon_QuickUVControl}.UV
    Tex=@n28
   ) {constMipValue=0, mipValueMode=TMVM_MipLevel, samplerType=SAMPLERTYPE_LinearColor}.R) {name=TextureTex_Trd, nodeColor=<rgba 0.3125 1 0 1>}
  StaticSwitchParameter(True=@n14, False=@n15) {parameterName=Use ToD Color Curve}
 ]
}
